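# Animation Blueprint for the male M_MED_BiomeHose cosmetic body, adding secondary jacket and collar motion to the character.
# It copies the attached parent's pose, runs a targeting-aware Control Rig, then chains spring-based Anim Dynamics into a cached pose.
# DefaultSlot and EmoteSlot layer over that from Anim_clock_Button, weighted by the front-end idle curve and emote state.
# The event-set simulation-space variable is never read by the graph.
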AnimBlueprint /BRCosmetics/Characters/Player/Male/Medium/Bodies/M_MED_BiomeHose/Meshes/M_MED_BiomeHose_AnimBP
target_skeleton=&/M_MED_BiomeHose_Skeleton
graphs [
 AnimGraph {nodes [
   Anim Dynamics - Bone: Dyn_Jacket_r_01!(SimulationSpace:AnimPhysSimSpaceType Enum, ComponentPose, Alpha=1, ExternalForce="0,0,0") -> (Pose) {
    angularDampingOverride=0.999
    angularSpringConstant=200
    bAngularSpring=true
    bOverrideAngularDamping=true
    bUseGravityOverride=true
    simulationSpace=RootRelative
    boundBone {boneName=Dyn_Jacket_r_01}
    componentAppliedLinearAccClamp=<4500 4500 5000>
    componentLinearAccScale=<0.5 0.5 0.6>
    componentLinearVelScale=<0.5 0.5 0.6>
    gravityOverride {z=-980}
    physicsBodyDefinitions [{
      boundBone {boneName=Dyn_Jacket_r_01}
      boxExtents=<10 10 10>
      constraintSetup {angularLimitsMax=<11 20 55>, angularLimitsMin=<-11 -15 4>, angularTarget=<1 0 0>}
      localJointOffset=<1.1 0 0>
     }]
    simSpaceSettings {maxAngularAcceleration=35, maxAngularVelocity=7, simSpaceAngularAlpha=1}
   }
   n1 = Slot 'DefaultSlot'()
   n2 = Use cached pose 'Base input'()
   n3 = Get Custom Idle is Playing Float() -> (Custom Idle is Playing Float:double)
   Output Pose(Result=Layered blend per bone(
     BasePose=Layered blend per bone(BasePose=@n2, BlendPoses_0=@n1, BlendWeights_0=@n3) {layerSetup [{branchFilters [{boneName=Anim_clock_Button}]}]}
     BlendPoses_0=Slot 'EmoteSlot'() {slotName=EmoteSlot}
     BlendWeights_0=Get Is Playing Custom Emote() -> (Is Playing Custom Emote:double)
    ) {layerSetup [{branchFilters [{boneName=Anim_clock_Button}]}]}) {layerGroup=None}
   n4 = Control Rig!(Source=Copy Pose From Mesh() {bUseAttachedParent=true}, Alpha, Targeting:bool=Get bIsTargeting()) -> (Pose) {controlRigAssetReference {blueprintRigClass=&/M_MED_BiomeHose_CtrlRig.M_MED_BiomeHose_CtrlRig_C}}
   n5 = Anim Dynamics - Bone: Dyn_Jacket_l_01(ComponentPose=Local To Component(LocalPose=@n4), Alpha=1, ExternalForce="0,0,0") {
    angularDampingOverride=0.999
    angularSpringConstant=200
    bAngularSpring=true
    bOverrideAngularDamping=true
    bUseGravityOverride=true
    simulationSpace=RootRelative
    boundBone {boneName=Dyn_Jacket_l_01}
    componentAppliedLinearAccClamp=<4500 4500 5000>
    componentLinearAccScale=<0.5 0.5 0.6>
    componentLinearVelScale=<0.5 0.5 0.6>
    gravityOverride {z=-980}
    physicsBodyDefinitions [{
      boundBone {boneName=Dyn_Jacket_l_01}
      boxExtents=<10 10 10>
      constraintSetup {angularLimitsMax=<10 4 55>, angularLimitsMin=<-11 -40 4>, angularTarget=<1 0 0>}
      localJointOffset=<1.1 0 0>
     }]
    simSpaceSettings {maxAngularAcceleration=35, maxAngularVelocity=7, simSpaceAngularAlpha=1}
   }
   n6 = Anim Dynamics - Bone: Dyn_Jacket_r_01!(ComponentPose=@n5, Alpha=1, ExternalForce="0,0,0") -> (Pose) {
    angularDampingOverride=0.999
    angularSpringConstant=200
    bAngularSpring=true
    bOverrideAngularDamping=true
    bUseGravityOverride=true
    simulationSpace=RootRelative
    boundBone {boneName=Dyn_Jacket_r_01}
    componentAppliedLinearAccClamp=<4500 4500 5000>
    componentLinearAccScale=<0.5 0.5 0.6>
    componentLinearVelScale=<0.5 0.5 0.6>
    gravityOverride {z=-980}
    physicsBodyDefinitions [{
      boundBone {boneName=Dyn_Jacket_r_01}
      boxExtents=<10 10 10>
      constraintSetup {angularLimitsMax=<5 20 55>, angularLimitsMin=<-5 -3 4>, angularTarget=<1 0 0>}
      localJointOffset=<1.1 0 0>
     }]
    simSpaceSettings {maxAngularAcceleration=35, maxAngularVelocity=7, simSpaceAngularAlpha=1}
   }
   n7 = Anim Dynamics - Bone: Dyn_Jacket_bk_r(ComponentPose=@n6, Alpha=1, ExternalForce="0,0,0") {
    angularDampingOverride=0.9
    angularSpringConstant=35
    bAngularSpring=true
    bOverrideAngularDamping=true
    bUseGravityOverride=true
    boundBone {boneName=Dyn_Jacket_bk_r}
    componentAppliedLinearAccClamp=<4500 4500 4500>
    componentLinearAccScale=<0.2 0.2 0.2>
    componentLinearVelScale=<0.2 0.2 0.2>
    gravityOverride {z=-980}
    physicsBodyDefinitions [{
      boundBone {boneName=Dyn_Jacket_bk_r}
      boxExtents=<10 10 10>
      constraintSetup {angularLimitsMax=<0 3 -3>, angularLimitsMin=<0 -3 -30>, angularTarget=<0 0 0>}
      localJointOffset=<0.9 0 0>
     }]
    simSpaceSettings {maxAngularAcceleration=35, maxAngularVelocity=7, simSpaceAngularAlpha=1}
   }
   n8 = Anim Dynamics - Bone: Dyn_Jacket_bk_l(ComponentPose=@n7, Alpha=1, ExternalForce="0,0,0") {
    angularDampingOverride=0.9
    angularSpringConstant=35
    bAngularSpring=true
    bOverrideAngularDamping=true
    bUseGravityOverride=true
    boundBone {boneName=Dyn_Jacket_bk_l}
    componentAppliedLinearAccClamp=<4500 4500 4500>
    componentLinearAccScale=<0.2 0.2 0.2>
    componentLinearVelScale=<0.2 0.2 0.2>
    gravityOverride {z=-980}
    physicsBodyDefinitions [{
      boundBone {boneName=Dyn_Jacket_bk_l}
      boxExtents=<10 10 10>
      constraintSetup {angularLimitsMax=<0 3 -3>, angularLimitsMin=<0 -3 -30>, angularTarget=<0 0 0>}
      localJointOffset=<0.9 0 0>
     }]
    simSpaceSettings {maxAngularAcceleration=35, maxAngularVelocity=7, simSpaceAngularAlpha=1}
   }
   n9 = Anim Dynamics - Bone: Dyn_Jacket_bk(ComponentPose=@n8, Alpha=1, ExternalForce="0,0,0") {
    angularDampingOverride=0.9
    angularSpringConstant=35
    bAngularSpring=true
    bOverrideAngularDamping=true
    bUseGravityOverride=true
    boundBone {boneName=Dyn_Jacket_bk}
    componentAppliedLinearAccClamp=<4500 4500 4500>
    componentLinearAccScale=<0.2 0.2 0.2>
    componentLinearVelScale=<0.2 0.2 0.2>
    gravityOverride {z=-980}
    physicsBodyDefinitions [{
      boundBone {boneName=Dyn_Jacket_bk}
      boxExtents=<10 10 10>
      constraintSetup {angularLimitsMax=<0 3 -3>, angularLimitsMin=<0 -3 -30>, angularTarget=<0 0 0>}
      localJointOffset=<0.9 0 0>
     }]
    simSpaceSettings {maxAngularAcceleration=35, maxAngularVelocity=7, simSpaceAngularAlpha=1}
   }
   n10 = Anim Dynamics - Bone: Dyn_Collar_l(ComponentPose=@n9, Alpha=1, ExternalForce="0,0,0") {
    angularDampingOverride=0.9
    angularSpringConstant=150
    bAngularSpring=true
    bOverrideAngularDamping=true
    bUseGravityOverride=true
    boundBone {boneName=Dyn_Collar_l}
    componentAppliedLinearAccClamp=<4500 4500 4500>
    componentLinearAccScale=<0.5 0.5 0.5>
    componentLinearVelScale=<0.5 0.5 0.5>
    gravityOverride {z=-980}
    physicsBodyDefinitions [{
      boundBone {boneName=Dyn_Collar_l}
      boxExtents=<8 8 8>
      constraintSetup {angularLimitsMax=<20 35 1>, angularLimitsMin=<-20 -10 -1>, angularTarget=<1 0 0>}
      localJointOffset=<-1 0 0>
     }]
    relativeSpaceBone {boneName=neck_02}
    simSpaceSettings {maxAngularAcceleration=35, maxAngularVelocity=7, simSpaceAngularAlpha=1}
   }
   n11 = Anim Dynamics - Bone: Dyn_Collar_r(ComponentPose=@n10, Alpha=1, ExternalForce="0,0,0") {
    angularDampingOverride=0.9
    angularSpringConstant=150
    bAngularSpring=true
    bOverrideAngularDamping=true
    bUseGravityOverride=true
    worldSpaceMinimumScale=0.1
    boundBone {boneName=Dyn_Collar_r}
    componentAppliedLinearAccClamp=<5000 5000 5000>
    componentLinearAccScale=<0.5 0.5 0.5>
    componentLinearVelScale=<0.5 0.5 0.5>
    gravityOverride {z=-980}
    physicsBodyDefinitions [{
      boundBone {boneName=Dyn_Collar_r}
      boxExtents=<8 8 8>
      constraintSetup {angularLimitsMax=<20 10 1>, angularLimitsMin=<-20 -35 -1>, angularTarget=<1 0 0>}
      localJointOffset=<-1 0 0>
     }]
    relativeSpaceBone {boneName=neck_02}
    simSpaceSettings {maxAngularAcceleration=35, maxAngularVelocity=7, simSpaceAngularAlpha=1}
   }
   Save cached pose 'Base input'(Pose=Component To Local(ComponentPose=@n11)) {cacheName=Base input}
  ]}
 EventGraph {nodes [
   Event BlueprintUpdateAnimation()
   > n1 = Sequence()
   > n2 = Branch(Condition=Get bIsPlayingEmote())
   > Set Is Playing Custom Emote(Is Playing Custom Emote:double=1) -> (Output_Get:double)
   Set Is Playing Custom Emote(execute=@n2.else, Is Playing Custom Emote=0)
   n3 = Branch(execute=@n1.then_1, Condition=Get bIsInFrontEnd())
   > Set SimlationSpace(SimlationSpace:AnimPhysSimSpaceType Enum=Component) -> (Output_Get:AnimPhysSimSpaceType Enum)
   > Set Custom Idle is Playing Float(Custom Idle is Playing Float:double=GetCurveValue(self=GetOwnerFrontEndAnimInstance(), CurveName=Frontend_Idle_Bypass)) -> (Output_Get:double)
   Set SimlationSpace(execute=@n3.else)
   Set Is Playing Custom Emote(Is Playing Custom Emote=GetCurveValue(CurveName=emote_BuiltIn))
  ]}
]
variables [
 Custom Idle is Playing Float {type=number}
 Is Playing Custom Emote {type=number}
 SimlationSpace {type=AnimPhysSimSpaceType}
]
