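# Post-body AnimBlueprint for the BiomeHose male face accessory, adding secondary hair motion on top of the pose copied from the attached parent mesh.
# Additive Modify Bone nodes offset Dyn_hair_bk, then chained spring-driven Anim Dynamics nodes simulate Dyn_hair_bk, Dyn_Hair_a and Dyn_Hair_b.
# The update event maps head-to-spine_05 yaw into sky and ground hair-correction rotators and sets NotSky from the skydiving state.
# The shown AnimGraph uses fixed values rather than reading these variables.
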AnimBlueprint /BRCosmetics/Characters/Player/Male/Medium/Bodies/M_MED_BiomeHose/Meshes/Parts/M_MED_BiomeHose_FaceAcc_AnimBP
target_skeleton=&/M_MED_BiomeHose_FaceAcc_Skeleton
graphs [
 AnimGraph {nodes [
   n1 = "Transform (Modify) Bone - Bone: Dyn_hair_bk"!(
    ComponentPose=Local To Component(LocalPose=Copy Pose From Mesh() {bUseAttachedParent=true})
    bAlphaBoolEnabled:bool=True
    Translation="0,1,2"
    Rotation="0.000,0.000,55.000"
    Scale="1,1,1"
   ) -> (Pose) {
    alphaInputType=Bool
    rotationMode=BMM_Additive
    translationMode=BMM_Additive
    boneToModify {boneName=Dyn_hair_bk}
    rotation {roll=55}
    translation {y=1, z=2}
   }
   n2 = "Transform (Modify) Bone - Bone: Dyn_hair_bk"(ComponentPose=@n1, bAlphaBoolEnabled=True, Translation="0,0,0", Rotation="0.000,0.000,55.000", Scale="1,1,1") {alphaInputType=Bool, rotationMode=BMM_Additive, translationMode=BMM_Additive, boneToModify {boneName=Dyn_hair_bk}, rotation {roll=55}}
   n3 = Anim Dynamics - Bone: Dyn_hair_bk(ComponentPose=@n2, Alpha=1, ExternalForce="0,0,0") {
    angularDampingOverride=0.9
    angularSpringConstant=250
    bAngularSpring=true
    bOverrideAngularDamping=true
    bUseGravityOverride=true
    boundBone {boneName=Dyn_hair_bk}
    componentAppliedLinearAccClamp=<5000 5000 5000>
    componentLinearAccScale=<0.15 0.15 0.15>
    componentLinearVelScale=<0.15 0.15 0.15>
    gravityOverride {z=-980}
    physicsBodyDefinitions [{
      boundBone {boneName=Dyn_hair_bk}
      boxExtents=<12 12 12>
      constraintSetup {angularLimitsMax=<0 0 1>, angularLimitsMin=<0 0 -10>, angularTarget=<1 0 0>}
      localJointOffset=<1 0 -0.5>
     }]
    simSpaceSettings {maxAngularAcceleration=35, maxAngularVelocity=7, simSpaceAngularAlpha=1}
   }
   n4 = Anim Dynamics - Bone: Dyn_Hair_a(ComponentPose=@n3, Alpha=1, ExternalForce="0,0,0") {
    angularDampingOverride=0.999
    angularSpringConstant=250
    bAngularSpring=true
    bOverrideAngularDamping=true
    bUseGravityOverride=true
    boundBone {boneName=Dyn_Hair_a}
    componentAppliedLinearAccClamp=<5000 5000 5000>
    componentLinearAccScale=<0.15 0.15 0.15>
    componentLinearVelScale=<0.15 0.15 0.15>
    gravityOverride {z=-980}
    physicsBodyDefinitions [{
      boundBone {boneName=Dyn_Hair_a}
      boxExtents=<12 12 12>
      constraintSetup {angularLimitsMax=<0 10 10>, angularLimitsMin=<0 -10 8>, angularTarget=<1 0 0>}
      localJointOffset=<1 0 -1.3>
     }]
    simSpaceSettings {maxAngularAcceleration=35, maxAngularVelocity=7, simSpaceAngularAlpha=1}
   }
   n5 = Anim Dynamics - Bone: Dyn_Hair_b(ComponentPose=@n4, Alpha=1, ExternalForce="0,0,0") {
    angularDampingOverride=0.999
    angularSpringConstant=150
    bAngularSpring=true
    bOverrideAngularDamping=true
    bUseGravityOverride=true
    boundBone {boneName=Dyn_Hair_b}
    componentAppliedLinearAccClamp=<5000 5000 5000>
    componentLinearAccScale=<0.15 0.15 0.15>
    componentLinearVelScale=<0.15 0.15 0.15>
    gravityOverride {z=-980}
    physicsBodyDefinitions [{
      boundBone {boneName=Dyn_Hair_b}
      boxExtents=<10 10 10>
      constraintSetup {angularLimitsMax=<0 10 10>, angularLimitsMin=<0 -10 8>, angularTarget=<1 0 0>}
      localJointOffset=<1 0 -1.3>
     }]
    simSpaceSettings {maxAngularAcceleration=35, maxAngularVelocity=7, simSpaceAngularAlpha=1}
   }
   Output Pose(Result=Component To Local(ComponentPose=@n5)) {layerGroup=None}
  ]}
 EventGraph {nodes [
   Event BlueprintUpdateAnimation()
   n1 = GetOwnerSkeletalMeshComponent()
   n2 = BreakRotator(InRot=BreakTransform(InTransform=GetDeltaTransformFromRefPose(self=@n1, BoneName=head, BaseName=spine_05)).Rotation)
   > Set Hair Corr Sky(Hair Corr Sky:Rotator=MakeRotator(Roll=MapRangeClamped(Value=@n2.Yaw, InRangeA=-10, InRangeB=-35, OutRangeB=70), Yaw=-5)) -> (Output_Get:Rotator)
   > Set Hair Corr Ground(Hair Corr Ground:Rotator=MakeRotator(Roll=MapRangeClamped(Value=@n2.Yaw, InRangeA=-10, InRangeB=-40, OutRangeB=45))) -> (Output_Get:Rotator)
   > Set NotSky(NotSky:bool=NOT Boolean(A=Get bIsSkyDiving())) -> (Output_Get:bool)
  ]}
]
variables [
 Hair Corr Ground {type=Rotator}
 Hair Corr Sky {type=Rotator}
 NotSky {type=boolean}
]
